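; Calendar event manager actor that adjusts which storm shapes are enabled and their weights when a tagged side effect is applied.
; It waits for playlist data, destroying itself on begin play if the hotfix toggle is off.
; The side effect stores its trigger tag; the authority then waits for the game state component's storm shape registry, binding its loaded delegate if needed.
; The tag keys a data map whose scalable-float enabled and weight values are written into that registry.
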(event EventBeginPlay
  (Class|FortAsyncActionCurrentPlaylistReady|PlaylistDataReadyAsync
    (:OnReady
      (if (not (ScalableFloat|GetValueasBool (Variables|Default|GetHF_CalendarEventsEnabled)))
        (Actor|DestroyActor)))))

(event SideEffects|EventBP_ApplySideEffect
  (params (ReactionSource :type Object%20Reference) (TargetObject :type Object%20Reference) (SideEffectTriggerTag :type Gameplay%20Tag%20Structure%20%28by%20ref%29) (SideEffectTriggerType :type EBlueprintableSideEffectTriggerType%20Enum%20%28by%20ref%29 :default "TriggeredAsTarget"))
  (Variables|Default|SetEventTag SideEffectTriggerTag)
  (Networking|SwitchHasAuthority
    (:Authority
      (bind _gamestate (Class|FortAsyncActionCurrentPlaylistReady|PlaylistDataReadyAsync)
        (:OnReady
          (if (ScalableFloat|GetValueasBool (Variables|Default|GetHF_CalendarEventsEnabled))
            (Utilities|IsValid (Actor|GetComponentbyClass _gamestate "/YokeHopUnkind/Gameplay/BP_YokeHopUnkindGSC.BP_YokeHopUnkindGSC_C")
              (:Is%20Valid
                (if (Variables|/YokeHopUnkind/Gameplay/BP_YokeHopUnkindGSC.BP_YokeHopUnkindGSC_C|GetbRegistryLoaded (Actor|GetComponentbyClass _gamestate "/YokeHopUnkind/Gameplay/BP_YokeHopUnkindGSC.BP_YokeHopUnkindGSC_C"))
                  (CallFunction|ApplyCalendarEventWeights (Variables|Default|GetEventTag) (Variables|/Script/YokeHopUnkindRuntime.FortGameStateComponent_YokeHopUnkind|GetStormShapeRegistryStrongReference (Actor|GetComponentbyClass _gamestate "/YokeHopUnkind/Gameplay/BP_YokeHopUnkindGSC.BP_YokeHopUnkindGSC_C")))
                  (else
                    (Default|AssignOnStormShapeRegistryLoaded (Actor|GetComponentbyClass _gamestate "/YokeHopUnkind/Gameplay/BP_YokeHopUnkindGSC.BP_YokeHopUnkindGSC_C") (delegate Custom|OnStormShapeRegistryLoaded_Event))))))))))))

(event Custom|OnStormShapeRegistryLoaded_Event
  (CallFunction|ApplyCalendarEventWeights (Variables|Default|GetEventTag) (Variables|/Script/YokeHopUnkindRuntime.FortGameStateComponent_YokeHopUnkind|GetStormShapeRegistryStrongReference (Actor|GetComponentbyClass (stale SideEffects|EventBP_ApplySideEffect _gamestate) "/YokeHopUnkind/Gameplay/BP_YokeHopUnkindGSC.BP_YokeHopUnkindGSC_C"))))

(fn ApplyCalendarEventWeights
  (params (EventTag :type Gameplay%20Tag%20Structure%20%28by%20ref%29) (Registry :type Safe%20Zone%20Shape%20Provider%20Registry%20Object%20Reference))
  (bind _calendareventdatamap (Variables|Default|GetCalendarEventDataMap))
  (bind (_value _returnvalue) (Utilities|Map|Find _calendareventdatamap EventTag))
  (bind _registry (Variables|GetRegistry))
  (if _returnvalue
    (for _array_element (Utilities|Struct|BreakYokeHopUnkindCalendarEventSet _value)
      (StormShape|SetShapeConfigEnabled _registry (Utilities|Struct|BreakYokeHopUnkindCalendarEventData _array_element) (ScalableFloat|GetValueasBool (.Enabled_2_3FFD62A548DB9AB72434638600ADF246 (Utilities|Struct|BreakYokeHopUnkindCalendarEventData _array_element))))
      (StormShape|SetShapeConfigWeight _registry (Utilities|Struct|BreakYokeHopUnkindCalendarEventData _array_element) (ScalableFloat|GetValueatLevel (.Weight_4_56575236435F3AFC4C5275B35AF7F3C0 (Utilities|Struct|BreakYokeHopUnkindCalendarEventData _array_element)))))))

(fn ConstructionScript)
